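# Looping in-air sound cue for the Galileo Kayak glider cosmetic.
# A looping wave player feeds a continuous modulator that maps the TurnAmount parameter to pitch and volume, so the loop responds to turning.
# A root modulator adds slight random pitch and volume variation.
# Spatialization uses the shared player glider thrust attenuation, alongside the glider loop audio properties sheet and glider sound class.
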SoundCue /BRCosmetics/Sounds/Gliders/Galileo_Kayak/Glider_Galileo_Kayak_Loop_Cue
root=Modulator_0
nodes [
 {
  id=ModulatorContinuous_0
  children [WavePlayer_0]
  properties {
   pitchModulationParams {default=0, maxInput=1.1, maxOutput=1.1, minInput=1, minOutput=1, parameterName=TurnAmount}
   volumeModulationParams {default=0, minOutput=0.7, parameterName=TurnAmount}
  }
 }
 {id=Modulator_0, children [ModulatorContinuous_0], properties {pitchMax=1.02, pitchMin=0.98, volumeMax=0.5, volumeMin=0.4}}
 {
  id=WavePlayer_0
  properties {bLooping=true, soundWaveAssetPtr=&/Athena_Glider_Galileo_Kayak_InAir_Loop}
 }
]
properties {
 attenuationSettings=&Fort_Player_Glider_Thrust_Athena_Att
 audioPropertiesSheet=&APS_Cosmetic_Glider_Loop
 maxDistance=12500
 soundClassObject=&Fort_Gliders_Land
 volumeMultiplier=1
}
